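# Niagara system for the Marinara emote that spawns puffs, blue glows, lights and twinkling sparkles at the character's feet.
# Paired left/right emitters sample foot positions from the skeletal mesh and burst only while that foot's System PuffCanSpawn flag is set, by gating Emitter.Age.
# Burst probability follows the clamped system LoopCount, suppressing spawns at count zero.
# GPU Tinkerbell bursts feed SparkleRainTrail emitters through particle reads, leaving falling, flickering trails.
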NiagaraSystem /BRCosmetics/Animation/Game/MainPlayer/Emotes/Marinara/FX/NS_Emote_Marinara
dependencies {
 usedDataInterfaces [
  &NiagaraDataInterfaceCurve
  &NiagaraDataInterfaceParticleRead
  /Script/Niagara.NiagaraDataInterfaceVectorField
  &NiagaraDataInterfaceSkeletalMesh
 ]
 usedDynamicInputs [
  &Divide_Float
  &Add_Integer
  /Niagara/DynamicInputs/Divide/Divide_Int
  &ModuloInt
  &SetBoolByFloatComparison
  &MakeFloatFromVector
 ]
 usedModules [
  &self:SystemSpawnScript.SystemScriptSource.SystemScriptGraph.NiagaraNodeAssignment_3.SetVariables_#1
  &SystemState
  &self:SystemSpawnScript.SystemScriptSource.SystemScriptGraph.NiagaraNodeAssignment_0.SetVariables_#2
  &self:SystemSpawnScript.SystemScriptSource.SystemScriptGraph.NiagaraNodeAssignment_1.SetVariables_#3
  &self:SystemSpawnScript.SystemScriptSource.SystemScriptGraph.NiagaraNodeAssignment_2.SetVariables_#4
  &self:SystemSpawnScript.SystemScriptSource.SystemScriptGraph.NiagaraNodeAssignment_5.SetVariables_#5
 ]
}
emitters {
 Empty {
  emitterUpdateScript [
   &EmitterState(Life Cycle Mode=::NewEnumerator1, Loop Behavior=::NewEnumerator1, Loop Duration Mode=::NewEnumerator1, UseLoopDelay=false)
   /SetVariables(Emitter.Age:NiagaraFloat=&MakeCustomFloatFromBool(Bool=$System.PuffCanSpawnRightFoot, True Float=$Emitter.Age, False Float=0)) {isSetParameters=true, module=SetVariables_#6}
   &SpawnBurst_Instantaneous(
    Age:NiagaraFloat=$Emitter.Age
    Spawn Probability=&ClampFloat(Float=&MakeFloatFromInt(Int=$System.LoopCount))
    Spawn Time=0.08
    Use Spawn Probability=true
    UseLoopCountLimit=false
   )
   &SpawnBurst_Instantaneous(
    Age:NiagaraFloat=$Emitter.Age
    Spawn Probability=&ClampFloat(Float=&MakeFloatFromInt(Int=$System.LoopCount))
    Spawn Time=0.095
    Use Spawn Probability=true
    UseLoopCountLimit=false
   ) {module=SpawnBurst_Instantaneous002}
  ]
  particleSpawnScript [
   &InitializeParticle(
    Color=<rgba 1 1 1 0.5>
    Sprite Rotation Mode=::NewEnumerator2
    Sprite Size Mode=::NewEnumerator3
    Uniform Sprite Size=&RandomRangeFloat(Minimum=70, Maximum=100)
    UsePositionOffset=false
    Write Material Random=true
   )
   &SampleSkeletalMesh(
    AdditionalYawPitchRoll=false
    Bone / Socket Index=1
    Bone Sampling Mode:ENiagaraBoneSamplingMode=::NewEnumerator4
    Sampled Position Offset:Vector3f=<0 0 -1>
    Skeletal Mesh:NiagaraDataInterfaceSkeletalMesh=$System.Skeletal Mesh
   )
  ]
  particleUpdateScript [
   &ParticleState()
   &DynamicMaterialParameters(
    Index 0 Param 1=&FloatFromCurve(
     FloatCurve=DataInterface() {propertyValues {Curve {keys [{arriveTangent=1.3374, interpMode=RCIM_Cubic, leaveTangent=1.3374, tangentMode=RCTM_User, time=0, value=1}]}}}
     CurveIndex=$Particles.Age
    )
    Index 0 Param 2=0.15
    Index 0 Param 3=&FloatFromCurve(
     FloatCurve=DataInterface() {propertyValues {Curve {keys [
         {arriveTangent=4.9025, interpMode=RCIM_Cubic, leaveTangent=4.9025, tangentMode=RCTM_User, time=0, value=0}
         {arriveTangent=4.4266, interpMode=RCIM_Cubic, tangentMode=RCTM_Break, time=0.5796, value=2}
        ]}}}
     CurveIndex=$Particles.Age
     Scale Curve=0.05
    )
    Index 0 Param 4=&FloatFromCurve(
     FloatCurve=DataInterface() {propertyValues {Curve {keys [
         {arriveTangent=-3.3445, interpMode=RCIM_Cubic, leaveTangent=-3.3445, tangentMode=RCTM_User, time=0, value=0.462}
         {arriveTangent=-1.1268, interpMode=RCIM_Cubic, leaveTangent=-1.2066, tangentMode=RCTM_Break, time=0.3394, value=-0.3632}
         {arriveTangent=-1.4939, interpMode=RCIM_Cubic, tangentMode=RCTM_Break, time=1.2, value=-1}
        ]}}}
     CurveIndex=$Particles.Age
    )
    Index 1 Param 1=0.8
    Index 1 Param 2=0.5
    Index 1 Param 4=3
    Module.Param0WriteEnabled=true
    Module.Param0WriteEnabled1=true
    Module.Param1WriteEnabled=true
    Module.Param1WriteEnabled1=true
    Module.Param2WriteEnabled=true
    Module.Param2WriteEnabled1=true
    Module.Param3WriteEnabled=true
    Module.Param3WriteEnabled1=true
    Write Parameter Index 1=true
   )
   &SpriteRotationRate(Rotation Rate=20)
   /SetVariables(Particles.SpriteFacing=&TransformVector(Vector=<0 0 1>, Source Space=::World, Destination Space=::Local)) {isSetParameters=true, module=SetVariables_#7}
   &ScaleSpriteSize(Uniform Curve Sprite Scale=DataInterface() {propertyValues {Curve {keys [
        {arriveTangent=4.1785, interpMode=RCIM_Cubic, leaveTangent=4.1785, tangentMode=RCTM_User, time=0, value=0}
        {arriveTangent=1.3778, interpMode=RCIM_Cubic, leaveTangent=1.3778, tangentMode=RCTM_User, time=0.1452, value=0.6}
        {arriveTangent=-0.0405, interpMode=RCIM_Cubic, leaveTangent=-0.0405, tangentMode=RCTM_User, time=1, value=1}
       ]}}})
   &ApplyOwnerScaleToAttributes(Scale Initial Sprite Size=true) {enabled=false}
   &SolveForcesAndVelocity(Clamp Velocity=false, Limit Acceleration=false)
   &ScaleColor(
    Scale Alpha=&FloatFromCurve(FloatCurve=DataInterface() {propertyValues {Curve {keys [<at 0.8 1>,<at 1 0>]}}}, Scale Curve=0.02)
    ScaleA=true
    ScaleRGB=true
   )
   /SetVariables(Particles.OffsetPosition:NiagaraPosition=&AddVectorToPosition(Position=$Particles.Position, Vector=$Particles.SampleSkeletalMesh.SampledPosition)) {isSetParameters=true, module=SetVariables_#8}
   &FN_Light_Attributes(
    Create Color Attribute=true
    Create Position Attribute=true
    Create Radius Attribute=true
    Position=&AddVectorToPosition(Position=<0 0 10>, Vector=Linked() {linkedVariable {name=Particles.OffsetPosition, type=NiagaraPosition}})
    Radius=100
    Scale Color=500
   )
  ]
  rendererClasses [&NiagaraSpriteRendererProperties, &NiagaraLightRendererProperties]
  renderers [
   {rendererClass=&NiagaraSpriteRendererProperties}
   {rendererClass=&NiagaraSpriteRendererProperties, rendererIndex=1}
   {rendererClass=&NiagaraLightRendererProperties, rendererIndex=2}
  ]
 }
 Empty001 {
  emitterUpdateScript [
   &EmitterState(Life Cycle Mode=::NewEnumerator1, Loop Behavior=::NewEnumerator1, Loop Duration Mode=::NewEnumerator1, UseLoopDelay=false)
   /SetVariables(Emitter.Age:NiagaraFloat=&MakeCustomFloatFromBool(Bool=$System.PuffCanSpawnLeftFoot, True Float=$Emitter.Age, False Float=0)) {isSetParameters=true, module=SetVariables_#6}
   &SpawnBurst_Instantaneous(
    Age:NiagaraFloat=$Emitter.Age
    Spawn Probability=&ClampFloat(Float=&MakeFloatFromInt(Int=$System.LoopCount))
    Spawn Time=0.08
    Use Spawn Probability=true
    UseLoopCountLimit=false
   )
   &SpawnBurst_Instantaneous(
    Age:NiagaraFloat=$Emitter.Age
    Spawn Probability=&ClampFloat(Float=&MakeFloatFromInt(Int=$System.LoopCount))
    Spawn Time=0.095
    Use Spawn Probability=true
    UseLoopCountLimit=false
   ) {module=SpawnBurst_Instantaneous001}
  ]
  particleSpawnScript [
   &InitializeParticle(
    Color=<rgba 1 1 1 0.5>
    Sprite Rotation Mode=::NewEnumerator2
    Sprite Size Mode=::NewEnumerator3
    Uniform Sprite Size=&RandomRangeFloat(Minimum=70, Maximum=100)
    UsePositionOffset=false
    Write Material Random=true
   )
   &SampleSkeletalMesh(
    AdditionalYawPitchRoll=false
    Bone Sampling Mode:ENiagaraBoneSamplingMode=::NewEnumerator4
    Sampled Position Offset:Vector3f=<0 0 -1>
    Skeletal Mesh:NiagaraDataInterfaceSkeletalMesh=$System.Skeletal Mesh
   )
  ]
  particleUpdateScript [
   &ParticleState()
   &DynamicMaterialParameters(
    Index 0 Param 1=&FloatFromCurve(
     FloatCurve=DataInterface() {propertyValues {Curve {keys [{arriveTangent=1.3374, interpMode=RCIM_Cubic, leaveTangent=1.3374, tangentMode=RCTM_User, time=0, value=1}]}}}
     CurveIndex=$Particles.Age
    )
    Index 0 Param 2=0.15
    Index 0 Param 3=&FloatFromCurve(
     FloatCurve=DataInterface() {propertyValues {Curve {keys [
         {arriveTangent=4.9025, interpMode=RCIM_Cubic, leaveTangent=4.9025, tangentMode=RCTM_User, time=0, value=0}
         {arriveTangent=4.4266, interpMode=RCIM_Cubic, tangentMode=RCTM_Break, time=0.5796, value=2}
        ]}}}
     CurveIndex=$Particles.Age
     Scale Curve=0.05
    )
    Index 0 Param 4=&FloatFromCurve(
     FloatCurve=DataInterface() {propertyValues {Curve {keys [
         {arriveTangent=-3.3445, interpMode=RCIM_Cubic, leaveTangent=-3.3445, tangentMode=RCTM_User, time=0, value=0.462}
         {arriveTangent=-1.1268, interpMode=RCIM_Cubic, leaveTangent=-1.2066, tangentMode=RCTM_Break, time=0.3394, value=-0.3632}
         {arriveTangent=-1.4939, interpMode=RCIM_Cubic, tangentMode=RCTM_Break, time=1.2, value=-1}
        ]}}}
     CurveIndex=$Particles.Age
    )
    Index 1 Param 1=0.8
    Index 1 Param 2=0.5
    Index 1 Param 4=3
    Module.Param0WriteEnabled=true
    Module.Param0WriteEnabled1=true
    Module.Param1WriteEnabled=true
    Module.Param1WriteEnabled1=true
    Module.Param2WriteEnabled=true
    Module.Param2WriteEnabled1=true
    Module.Param3WriteEnabled=true
    Module.Param3WriteEnabled1=true
    Write Parameter Index 1=true
   )
   &SpriteRotationRate(Rotation Rate=20)
   /SetVariables(Particles.SpriteFacing=&TransformVector(Vector=<0 0 1>, Source Space=::World, Destination Space=::Local)) {isSetParameters=true, module=SetVariables_#7}
   &ScaleSpriteSize(Uniform Curve Sprite Scale=DataInterface() {propertyValues {Curve {keys [
        {arriveTangent=4.1785, interpMode=RCIM_Cubic, leaveTangent=4.1785, tangentMode=RCTM_User, time=0, value=0}
        {arriveTangent=1.3778, interpMode=RCIM_Cubic, leaveTangent=1.3778, tangentMode=RCTM_User, time=0.1452, value=0.6}
        {arriveTangent=-0.0405, interpMode=RCIM_Cubic, leaveTangent=-0.0405, tangentMode=RCTM_User, time=1, value=1}
       ]}}})
   &ApplyOwnerScaleToAttributes(Scale Initial Sprite Size=true) {enabled=false}
   &SolveForcesAndVelocity(Clamp Velocity=false, Limit Acceleration=false)
   &ScaleColor(
    Scale Alpha=&FloatFromCurve(
     FloatCurve=DataInterface() {propertyValues {Curve {keys [
         {interpMode=RCIM_Cubic, leaveTangent=-8.3587, tangentMode=RCTM_Break, time=0.8, value=1}
         {arriveTangent=-1.6308, interpMode=RCIM_Cubic, leaveTangent=-1.6308, tangentMode=RCTM_User, time=1, value=0}
        ]}}}
     Scale Curve=0.02
    )
    ScaleA=true
    ScaleRGB=true
   )
   /SetVariables(Particles.OffsetPosition:NiagaraPosition=&AddVectorToPosition(Position=$Particles.Position, Vector=$Particles.SampleSkeletalMesh.SampledPosition)) {isSetParameters=true, module=SetVariables_#8}
   &FN_Light_Attributes(
    Create Color Attribute=true
    Create Position Attribute=true
    Create Radius Attribute=true
    Position=&AddVectorToPosition(Position=<0 0 10>, Vector=Linked() {linkedVariable {name=Particles.OffsetPosition, type=NiagaraPosition}})
    Radius=500
    Scale Color=500
   )
   /SetVariables(Particles.LightIntensity:NiagaraFloat=&Lerp_Float(A=0, B=100, Alpha=1)) {isSetParameters=true, module=SetVariables_#9}
  ]
  rendererClasses [
   &NiagaraSpriteRendererProperties
   &NiagaraLightRendererProperties
   &NiagaraComponentRendererProperties
  ]
  renderers [
   {rendererClass=&NiagaraSpriteRendererProperties}
   {rendererClass=&NiagaraSpriteRendererProperties, rendererIndex=1}
   {rendererClass=&NiagaraLightRendererProperties, rendererIndex=2}
   {rendererClass=&NiagaraComponentRendererProperties, rendererIndex=3}
  ]
 }
 Empty002 {
  emitterUpdateScript [
   &EmitterState(Life Cycle Mode=::NewEnumerator1, Loop Behavior=::NewEnumerator1, Loop Duration Mode=::NewEnumerator1, UseLoopDelay=false)
   /SetVariables(Emitter.Age:NiagaraFloat=&MakeCustomFloatFromBool(Bool=$System.PuffCanSpawnRightFoot, True Float=$Emitter.Age, False Float=0)) {isSetParameters=true, module=SetVariables_#6}
   &SpawnBurst_Instantaneous(
    Age:NiagaraFloat=$Emitter.Age
    Spawn Probability=&ClampFloat(Float=&MakeFloatFromInt(Int=$System.LoopCount))
    Spawn Time=0.08
    Use Spawn Probability=true
    UseLoopCountLimit=false
   )
  ]
  particleSpawnScript [
   &InitializeParticle(
    Color=<rgba 0.4914 0.5337 1 0.08>
    Lifetime=0.7
    Sprite Rotation Mode=::NewEnumerator2
    Sprite Size Mode=::NewEnumerator3
    Uniform Sprite Size=100
    UsePositionOffset=false
    Write Material Random=true
   )
   &SampleSkeletalMesh(
    AdditionalYawPitchRoll=false
    Bone / Socket Index=1
    Bone Sampling Mode:ENiagaraBoneSamplingMode=::NewEnumerator4
    Sampled Position Offset:Vector3f=<0 0 -3>
    Skeletal Mesh:NiagaraDataInterfaceSkeletalMesh=$System.Skeletal Mesh
   )
  ]
  particleUpdateScript [
   &ParticleState()
   &DynamicMaterialParameters(
    Index 0 Param 1=&FloatFromCurve(
     FloatCurve=DataInterface() {propertyValues {Curve {keys [{arriveTangent=1.3374, interpMode=RCIM_Cubic, leaveTangent=1.3374, tangentMode=RCTM_User, time=0, value=1}]}}}
     CurveIndex=$Particles.Age
    )
    Index 0 Param 2=0.15
    Index 0 Param 3=&FloatFromCurve(
     FloatCurve=DataInterface() {propertyValues {Curve {keys [
         {arriveTangent=4.9025, interpMode=RCIM_Cubic, leaveTangent=4.9025, tangentMode=RCTM_User, time=0, value=0}
         {arriveTangent=4.4266, interpMode=RCIM_Cubic, tangentMode=RCTM_Break, time=0.5796, value=2}
        ]}}}
     CurveIndex=$Particles.Age
     Scale Curve=0.05
    )
    Index 0 Param 4=&FloatFromCurve(
     FloatCurve=DataInterface() {propertyValues {Curve {keys [
         {arriveTangent=-3.3445, interpMode=RCIM_Cubic, leaveTangent=-3.3445, tangentMode=RCTM_User, time=0, value=0.462}
         {arriveTangent=-1.1268, interpMode=RCIM_Cubic, leaveTangent=-1.2066, tangentMode=RCTM_Break, time=0.3394, value=-0.3632}
         {arriveTangent=-1.4939, interpMode=RCIM_Cubic, tangentMode=RCTM_Break, time=1.2, value=-1}
        ]}}}
     CurveIndex=$Particles.Age
    )
    Index 1 Param 1=0.8
    Index 1 Param 2=0.5
    Index 1 Param 4=3
    Module.Param0WriteEnabled=true
    Module.Param0WriteEnabled1=true
    Module.Param1WriteEnabled=true
    Module.Param1WriteEnabled1=true
    Module.Param2WriteEnabled=true
    Module.Param2WriteEnabled1=true
    Module.Param3WriteEnabled=true
    Module.Param3WriteEnabled1=true
    Write Parameter Index 1=true
   )
   &SpriteRotationRate(Rotation Rate=20)
   /SetVariables(Particles.SpriteFacing=&TransformVector(Vector=<0 0 1>, Source Space=::World, Destination Space=::Local)) {isSetParameters=true, module=SetVariables_#7}
   &ScaleSpriteSize(Uniform Curve Sprite Scale=DataInterface() {propertyValues {Curve {keys [
        {arriveTangent=4.1785, interpMode=RCIM_Cubic, leaveTangent=4.1785, tangentMode=RCTM_User, time=0, value=0}
        {arriveTangent=1.3778, interpMode=RCIM_Cubic, leaveTangent=1.3778, tangentMode=RCTM_User, time=0.1452, value=0.6}
        {arriveTangent=-0.0405, interpMode=RCIM_Cubic, leaveTangent=-0.0405, tangentMode=RCTM_User, time=1, value=1}
       ]}}})
   &ApplyOwnerScaleToAttributes(Scale Initial Sprite Size=true) {enabled=false}
   &SolveForcesAndVelocity(Clamp Velocity=false, Limit Acceleration=false)
   &ScaleColor(
    Scale Alpha=&FloatFromCurve(FloatCurve=DataInterface() {propertyValues {Curve {keys [
         {interpMode=RCIM_Cubic, tangentMode=RCTM_User, time=0, value=0}
         {interpMode=RCIM_Cubic, tangentMode=RCTM_User, time=0.5, value=1}
         {tangentMode=RCTM_User, time=1, value=0}
        ]}}})
    ScaleA=true
    ScaleRGB=true
   )
   /SetVariables(Particles.OffsetPosition:NiagaraPosition=&AddVectorToPosition(Position=$Particles.Position, Vector=$Particles.SampleSkeletalMesh.SampledPosition)) {isSetParameters=true, module=SetVariables_#8}
   &FN_Light_Attributes(
    Color=<rgba 0.3437 0.4089 1 1>
    Create Color Attribute=true
    Create Position Attribute=true
    Create Radius Attribute=true
    Position=&AddVectorToPosition(Position=<0 0 15>, Vector=Linked() {linkedVariable {name=Particles.OffsetPosition, type=NiagaraPosition}})
    Radius=100
    Scale Color=25
   )
   /SetVariables(
    Particles.ColorLight:LinearColor=<rgba 0.3437 0.4089 1 1>
    Particles.LightIntensity:NiagaraFloat=&Lerp_Float(
     A=0
     B=35
     Alpha=&FloatFromCurve(FloatCurve=DataInterface() {propertyValues {Curve {keys [
          {arriveTangent=5.5909, interpMode=RCIM_Cubic, leaveTangent=5.5909, tangentMode=RCTM_User, time=0, value=0}
          {interpMode=RCIM_Cubic, tangentMode=RCTM_User, time=0.5, value=1}
          {arriveTangent=-5.2859, tangentMode=RCTM_User, time=1, value=0}
         ]}}})
    )
   ) {isSetParameters=true, module=SetVariables_#9}
  ]
  rendererClasses [
   &NiagaraSpriteRendererProperties
   &NiagaraLightRendererProperties
   &NiagaraComponentRendererProperties
  ]
  renderers [
   {rendererClass=&NiagaraSpriteRendererProperties}
   {rendererClass=&NiagaraLightRendererProperties, rendererIndex=1}
   {rendererClass=&NiagaraComponentRendererProperties, rendererIndex=2}
   {rendererClass=&NiagaraSpriteRendererProperties, rendererIndex=3}
  ]
 }
 Empty003 {
  emitterUpdateScript [
   &EmitterState(Life Cycle Mode=::NewEnumerator1, Loop Behavior=::NewEnumerator1, Loop Duration Mode=::NewEnumerator1, UseLoopDelay=false)
   /SetVariables(Emitter.Age:NiagaraFloat=&MakeCustomFloatFromBool(Bool=$System.PuffCanSpawnLeftFoot, True Float=$Emitter.Age, False Float=0)) {isSetParameters=true, module=SetVariables_#6}
   &SpawnBurst_Instantaneous(
    Age:NiagaraFloat=$Emitter.Age
    Spawn Probability=&ClampFloat(Float=&MakeFloatFromInt(Int=$System.LoopCount))
    Spawn Time=0.08
    Use Spawn Probability=true
    UseLoopCountLimit=false
   )
  ]
  particleSpawnScript [
   &InitializeParticle(
    Color=<rgba 0.4914 0.5337 1 0.08>
    Lifetime=0.7
    Sprite Rotation Mode=::NewEnumerator2
    Sprite Size Mode=::NewEnumerator3
    Uniform Sprite Size=100
    UsePositionOffset=false
    Write Material Random=true
   )
   &SampleSkeletalMesh(
    AdditionalYawPitchRoll=false
    Bone Sampling Mode:ENiagaraBoneSamplingMode=::NewEnumerator4
    Sampled Position Offset:Vector3f=<0 0 -3>
    Skeletal Mesh:NiagaraDataInterfaceSkeletalMesh=$System.Skeletal Mesh
   )
  ]
  particleUpdateScript [
   &ParticleState()
   &DynamicMaterialParameters(
    Index 0 Param 1=&FloatFromCurve(
     FloatCurve=DataInterface() {propertyValues {Curve {keys [{arriveTangent=1.3374, interpMode=RCIM_Cubic, leaveTangent=1.3374, tangentMode=RCTM_User, time=0, value=1}]}}}
     CurveIndex=$Particles.Age
    )
    Index 0 Param 2=0.15
    Index 0 Param 3=&FloatFromCurve(
     FloatCurve=DataInterface() {propertyValues {Curve {keys [
         {arriveTangent=4.9025, interpMode=RCIM_Cubic, leaveTangent=4.9025, tangentMode=RCTM_User, time=0, value=0}
         {arriveTangent=4.4266, interpMode=RCIM_Cubic, tangentMode=RCTM_Break, time=0.5796, value=2}
        ]}}}
     CurveIndex=$Particles.Age
     Scale Curve=0.05
    )
    Index 0 Param 4=&FloatFromCurve(
     FloatCurve=DataInterface() {propertyValues {Curve {keys [
         {arriveTangent=-3.3445, interpMode=RCIM_Cubic, leaveTangent=-3.3445, tangentMode=RCTM_User, time=0, value=0.462}
         {arriveTangent=-1.1268, interpMode=RCIM_Cubic, leaveTangent=-1.2066, tangentMode=RCTM_Break, time=0.3394, value=-0.3632}
         {arriveTangent=-1.4939, interpMode=RCIM_Cubic, tangentMode=RCTM_Break, time=1.2, value=-1}
        ]}}}
     CurveIndex=$Particles.Age
    )
    Index 1 Param 1=0.8
    Index 1 Param 2=0.5
    Index 1 Param 4=3
    Module.Param0WriteEnabled=true
    Module.Param0WriteEnabled1=true
    Module.Param1WriteEnabled=true
    Module.Param1WriteEnabled1=true
    Module.Param2WriteEnabled=true
    Module.Param2WriteEnabled1=true
    Module.Param3WriteEnabled=true
    Module.Param3WriteEnabled1=true
    Write Parameter Index 1=true
   )
   &SpriteRotationRate(Rotation Rate=20)
   /SetVariables(Particles.SpriteFacing=&TransformVector(Vector=<0 0 1>, Source Space=::World, Destination Space=::Local)) {isSetParameters=true, module=SetVariables_#7}
   &ScaleSpriteSize(Uniform Curve Sprite Scale=DataInterface() {propertyValues {Curve {keys [
        {arriveTangent=4.1785, interpMode=RCIM_Cubic, leaveTangent=4.1785, tangentMode=RCTM_User, time=0, value=0}
        {arriveTangent=1.3778, interpMode=RCIM_Cubic, leaveTangent=1.3778, tangentMode=RCTM_User, time=0.1452, value=0.6}
        {arriveTangent=-0.0405, interpMode=RCIM_Cubic, leaveTangent=-0.0405, tangentMode=RCTM_User, time=1, value=1}
       ]}}})
   &ApplyOwnerScaleToAttributes(Scale Initial Sprite Size=true) {enabled=false}
   &SolveForcesAndVelocity(Clamp Velocity=false, Limit Acceleration=false)
   &ScaleColor(
    Scale Alpha=&FloatFromCurve(FloatCurve=DataInterface() {propertyValues {Curve {keys [
         {interpMode=RCIM_Cubic, tangentMode=RCTM_User, time=0, value=0}
         {interpMode=RCIM_Cubic, tangentMode=RCTM_User, time=0.5, value=1}
         {tangentMode=RCTM_User, time=1, value=0}
        ]}}})
    ScaleA=true
    ScaleRGB=true
   )
   /SetVariables(Particles.OffsetPosition:NiagaraPosition=&AddVectorToPosition(Position=$Particles.Position, Vector=$Particles.SampleSkeletalMesh.SampledPosition)) {isSetParameters=true, module=SetVariables_#8}
   &FN_Light_Attributes(
    Color=<rgba 0.3437 0.4089 1 1>
    Create Color Attribute=true
    Create Position Attribute=true
    Create Radius Attribute=true
    Position=&AddVectorToPosition(Position=<0 0 15>, Vector=Linked() {linkedVariable {name=Particles.OffsetPosition, type=NiagaraPosition}})
    Radius=75
    Scale Color=25
   )
   /SetVariables(
    Particles.ColorLight:LinearColor=<rgba 0.3437 0.4089 1 1>
    Particles.LightIntensity:NiagaraFloat=&Lerp_Float(
     A=0
     B=35
     Alpha=&FloatFromCurve(FloatCurve=DataInterface() {propertyValues {Curve {keys [
          {arriveTangent=5.5909, interpMode=RCIM_Cubic, leaveTangent=5.5909, tangentMode=RCTM_User, time=0, value=0}
          {interpMode=RCIM_Cubic, tangentMode=RCTM_User, time=0.5, value=1}
          {arriveTangent=-5.2859, tangentMode=RCTM_User, time=1, value=0}
         ]}}})
    )
   ) {isSetParameters=true, module=SetVariables_#9}
  ]
  rendererClasses [
   &NiagaraSpriteRendererProperties
   &NiagaraLightRendererProperties
   &NiagaraComponentRendererProperties
  ]
  renderers [
   {rendererClass=&NiagaraSpriteRendererProperties}
   {rendererClass=&NiagaraLightRendererProperties, rendererIndex=1}
   {rendererClass=&NiagaraComponentRendererProperties, rendererIndex=2}
  ]
 }
 SparkleRainTrail001 {
  simTarget=GPUComputeSim
  emitterUpdateScript [
   &EmitterState(Life Cycle Mode=::NewEnumerator1, Loop Behavior=::NewEnumerator1, Loop Duration Mode=::NewEnumerator1, UseLoopDelay=false)
   &SpawnParticlesFromOtherEmitter(
    Attribute Reader=DataInterface() {propertyValues {EmitterBinding {bindingMode=Other, emitterName=Tinkerbell002}}}
    Maximum Age Spawn Threshold=2
    Minimum Age Spawn Threshold=0
    Spawn Rate Per Particle Cap=22
    Use Maximum Age=false
    Use Minimum Age=false
   )
  ]
  particleSpawnScript [
   &InitializeParticle(
    Color=&RandomRangeLinearColor(Minimum=<rgba 0.4886 0.5173 1 1>, Maximum=<rgba 0.1968 0.249 1 1>)
    Lifetime Max=1
    Lifetime Min=0.333
    Lifetime Mode=::NewEnumerator1
    Sprite Rotation Angle=0.125
    Sprite Rotation Mode=::NewEnumerator1
    Sprite Size Mode=::NewEnumerator1
    Uniform Sprite Size Max=11
    Uniform Sprite Size Min=11
    UsePositionOffset=false
    Write Material Random=true
   )
   &SampleParticlesFromOtherEmitter(
    Applied Velocity Scale:NiagaraFloat=&FloatFromCurve(FloatCurve=DataInterface() {propertyValues {Curve {keys [<at 0 1>,<at 1 0>]}}}, Scale Curve=0.5)
    Maximum Sampled Particle Age=2
    Minimum Sampled Particle Age=0
    Module.MaxSampledAgeCutoff=true
    Module.MinSampledAgeCutoff=true
    Reject Particle By Sampled Age=true
    Velocity Sampling=::NewEnumerator0
   )
   &ShapeLocation(MaskSpawnGroup=false, Override Seed=false, Sphere Radius=5, UseOwnerScale=true)
  ]
  particleUpdateScript [
   &ParticleState()
   &Drag(Drag=2, Use Linear Drag=true, Use Rotational Drag=false)
   &GravityForce(Gravity=&RandomRangeVector(Minimum=<0 0 -50>, Maximum=<1 1 -100>))
   &CurlNoiseForce(Noise Frequency=1, Noise Strength=15, Pan Noise=false)
   &SolveForcesAndVelocity(Clamp Velocity=false, Limit Acceleration=false)
   &ScaleColor(
    Scale Alpha=&FloatFromCurve(
     FloatCurve=DataInterface() {propertyValues {Curve {keys [
         {tangentMode=RCTM_User, time=0, value=1}
         {arriveTangent=-0.0405, interpMode=RCIM_Cubic, leaveTangent=-0.0405, tangentMode=RCTM_User, time=0.7009, value=1}
         {arriveTangent=-8.0089, interpMode=RCIM_Cubic, leaveTangent=-8.0089, tangentMode=RCTM_User, time=1, value=0}
        ]}}}
     Scale Curve=&Waveform(
      Global Phase=&Add_Float(A=$Particles.Age, B=&RandomRangeFloat(Minimum=0, Maximum=1))
      "Waveform [1]"=::NewEnumerator0
      "[1] Frequency"=&RandomRangeFloat(Minimum=3.618, Maximum=11.08)
      "[1] Amplitude Min/Max"=<0 1>
     )
    )
    Scale RGB=&VectorFromFloat(Value=&Lerp_Float(
      A=0.25
      B=22
      Alpha=&FloatFromCurve(FloatCurve=DataInterface() {propertyValues {Curve {keys [<at 0 1>,<at 0.333 0>]}}})
     ))
    ScaleA=true
    ScaleRGB=true
   )
   &ScaleSpriteSize(Uniform Curve Sprite Scale=DataInterface() {propertyValues {Curve {keys [
        {interpMode=RCIM_Cubic, tangentMode=RCTM_User, time=0, value=0}
        {interpMode=RCIM_Cubic, tangentMode=RCTM_User, time=0.5, value=1}
        {tangentMode=RCTM_User, time=1, value=0}
       ]}}})
  ]
  rendererClasses [&NiagaraSpriteRendererProperties]
  renderers [{rendererClass=&NiagaraSpriteRendererProperties}]
 }
 SparkleRainTrail002 {
  simTarget=GPUComputeSim
  emitterUpdateScript [
   &EmitterState(Life Cycle Mode=::NewEnumerator1, Loop Behavior=::NewEnumerator1, Loop Duration Mode=::NewEnumerator1, UseLoopDelay=false)
   &SpawnParticlesFromOtherEmitter(
    Attribute Reader=DataInterface() {propertyValues {EmitterBinding {bindingMode=Other, emitterName=Tinkerbell001}}}
    Maximum Age Spawn Threshold=2
    Minimum Age Spawn Threshold=0
    Spawn Rate Per Particle Cap=22
    Use Maximum Age=false
    Use Minimum Age=false
   )
  ]
  particleSpawnScript [
   &InitializeParticle(
    Color=&RandomRangeLinearColor(Minimum=<rgba 0.4886 0.5173 1 1>, Maximum=<rgba 0.1968 0.249 1 1>)
    Lifetime Max=1
    Lifetime Min=0.333
    Lifetime Mode=::NewEnumerator1
    Sprite Rotation Angle=0.125
    Sprite Rotation Mode=::NewEnumerator1
    Sprite Size Mode=::NewEnumerator1
    Uniform Sprite Size Max=11
    Uniform Sprite Size Min=11
    UsePositionOffset=false
    Write Material Random=true
   )
   &SampleParticlesFromOtherEmitter(
    Applied Velocity Scale:NiagaraFloat=&FloatFromCurve(FloatCurve=DataInterface() {propertyValues {Curve {keys [<at 0 1>,<at 1 0>]}}}, Scale Curve=0.5)
    Maximum Sampled Particle Age=2
    Minimum Sampled Particle Age=0
    Module.MaxSampledAgeCutoff=true
    Module.MinSampledAgeCutoff=true
    Reject Particle By Sampled Age=true
    Velocity Sampling=::NewEnumerator0
   )
   &ShapeLocation(MaskSpawnGroup=false, Override Seed=false, Sphere Radius=5, UseOwnerScale=true)
  ]
  particleUpdateScript [
   &ParticleState()
   &Drag(Drag=2, Use Linear Drag=true, Use Rotational Drag=false)
   &GravityForce(Gravity=&RandomRangeVector(Minimum=<0 0 -50>, Maximum=<1 1 -100>))
   &CurlNoiseForce(Noise Frequency=1, Noise Strength=15, Pan Noise=false)
   &SolveForcesAndVelocity(Clamp Velocity=false, Limit Acceleration=false)
   &ScaleColor(
    Scale Alpha=&FloatFromCurve(
     FloatCurve=DataInterface() {propertyValues {Curve {keys [
         {tangentMode=RCTM_User, time=0, value=1}
         {arriveTangent=-0.0405, interpMode=RCIM_Cubic, leaveTangent=-0.0405, tangentMode=RCTM_User, time=0.7009, value=1}
         {arriveTangent=-8.0089, interpMode=RCIM_Cubic, leaveTangent=-8.0089, tangentMode=RCTM_User, time=1, value=0}
        ]}}}
     Scale Curve=&Waveform(
      Global Phase=&Add_Float(A=$Particles.Age, B=&RandomRangeFloat(Minimum=0, Maximum=1))
      "Waveform [1]"=::NewEnumerator0
      "[1] Frequency"=&RandomRangeFloat(Minimum=3.618, Maximum=11.08)
      "[1] Amplitude Min/Max"=<0 1>
     )
    )
    Scale RGB=&VectorFromFloat(Value=&Lerp_Float(
      A=0.25
      B=22
      Alpha=&FloatFromCurve(FloatCurve=DataInterface() {propertyValues {Curve {keys [<at 0 1>,<at 0.333 0>]}}})
     ))
    ScaleA=true
    ScaleRGB=true
   )
   &ScaleSpriteSize(Uniform Curve Sprite Scale=DataInterface() {propertyValues {Curve {keys [
        {interpMode=RCIM_Cubic, tangentMode=RCTM_User, time=0, value=0}
        {interpMode=RCIM_Cubic, tangentMode=RCTM_User, time=0.5, value=1}
        {tangentMode=RCTM_User, time=1, value=0}
       ]}}})
  ]
  rendererClasses [&NiagaraSpriteRendererProperties]
  renderers [{rendererClass=&NiagaraSpriteRendererProperties}]
 }
 Tinkerbell001 {
  simTarget=GPUComputeSim
  emitterUpdateScript [
   &EmitterState(Life Cycle Mode=::NewEnumerator1, Loop Behavior=::NewEnumerator1, Loop Delay=7, Loop Duration Mode=::NewEnumerator1, UseLoopDelay=false)
   /SetVariables(Emitter.Age:NiagaraFloat=&MakeCustomFloatFromBool(Bool=$System.PuffCanSpawnRightFoot, True Float=$Emitter.Age, False Float=0)) {isSetParameters=true, module=SetVariables_#10}
   &SpawnRate(SpawnRate=20, Use Spawn Probability=false) {enabled=false}
   &SpawnBurst_Instantaneous(
    Age:NiagaraFloat=$Emitter.Age
    Spawn Count=5
    Spawn Probability=&ClampFloat(Float=&MakeFloatFromInt(Int=$System.LoopCount))
    Spawn Time=0.08
    Use Spawn Probability=true
    UseLoopCountLimit=false
   )
  ]
  particleSpawnScript [
   &PartitionParticles() {enabled=false}
   &InitializeParticle(
    Color=<rgba 0.5349 0.5662 1 0.49>
    Lifetime=0.75
    Position Offset=<0 0 100>
    Sprite Rotation Mode=::NewEnumerator2
    Sprite Size Mode=::NewEnumerator3
    Uniform Sprite Size=20
    UsePositionOffset=true
    Write Material Random=true
   )
   &SkeletalMeshLocation(Bone / Socket Index=1, Bone Sampling Mode=::NewEnumerator4, MaskSpawnGroup=false, Skeletal Mesh=$System.Skeletal Mesh)
   &ShapeLocation(
    Cylinder Height=4
    Cylinder Height Midpoint=0
    Cylinder Radius=15
    MaskSpawnGroup=false
    Offset=<0 0 -8>
    Offset Mode=::NewEnumerator0
    Override Seed=false
    Rotation Axis=<0 0 1>
    Rotation Mode=::NewEnumerator1
    Shape Origin=$Output.SkeletalMeshLocation.SampledPositionWithoutOffset
    Shape Primitive=::NewEnumerator1
    Surface Only=true
    Surface Only Band Thickness=0.5
    Use Endcaps In Surface Only Mode=false
    UseOwnerScale=true
   )
   &AddVelocity(
    Cone Angle=75
    Cone Axis=<0 0 1>
    Module.UseConeFalloff=true
    Module.UseVelDistribution=true
    Override Seed=false
    Velocity Mode=::NewEnumerator2
    Velocity Speed=&RandomRangeFloat(Minimum=75, Maximum=100)
   )
   /SetVariables(Particles.Velocity=&Multiply_VectorByFloat(
     Vector=&Multiply_Vector(A=$Particles.ShapeLocation.ShapeNormal, B=<1 1 500>)
     Float=&RandomRangeFloat(Minimum=10, Maximum=40)
    )) {enabled=false, isSetParameters=true, module=SetVariables_#11}
  ]
  particleUpdateScript [
   /Niagara/VectorFields/SampleVectorField(
    FieldRotate:Quat4f=<0 0 1 1>
    Sampled Vector Scale=0.05
    VectorField=DataInterface() {propertyValues {Field=/Game/Effects/VectorFields/Noisy_Turbulence_103, bTileX=true, bTileY=true, bTileZ=true}}
   ) {enabled=false}
   &ParticleState()
   &SpriteRotationRate(Rotation Rate=&MaskFloatBySpawnGroup(Masked Float=45, Passthrough Float=-45))
   &ScaleSpriteSize(Uniform Curve Sprite Scale=DataInterface() {propertyValues {Curve {keys [
        {arriveTangent=9.0725, interpMode=RCIM_Cubic, leaveTangent=9.0725, tangentMode=RCTM_User, time=0, value=0}
        {arriveTangent=-0.0039, interpMode=RCIM_Cubic, leaveTangent=-0.0039, tangentMode=RCTM_User, time=0.1408, value=0.9847}
        {arriveTangent=-4.4862, interpMode=RCIM_Cubic, leaveTangent=-4.4862, tangentMode=RCTM_User, time=1, value=0}
       ]}}})
   &ScaleColor(
    Scale Alpha=&FloatFromCurve(
     FloatCurve=DataInterface() {propertyValues {Curve {keys [{interpMode=RCIM_Cubic, tangentMode=RCTM_User, time=0, value=1}, {interpMode=RCIM_Cubic, tangentMode=RCTM_User, time=1, value=0}]}}}
     Scale Curve=&Waveform(
      Global Phase=&Add_Float(A=$Particles.Age, B=&RandomRangeFloat(Minimum=0, Maximum=1))
      "Waveform [1]"=::NewEnumerator0
      "[1] Frequency"=&RandomRangeFloat(Minimum=1.618, Maximum=3.618)
      "[1] Amplitude Min/Max"=<0 1>
     )
    )
    Scale RGB=&VectorFromFloat(Value=&Lerp_Float(
      A=1
      B=3
      Alpha=&FloatFromCurve(FloatCurve=DataInterface() {propertyValues {Curve {keys [<at 0 1>,<at 0.25 0>]}}})
     ))
    ScaleA=true
    ScaleRGB=true
   )
   &DynamicMaterialParameters(
    Index 0 Param 1=1
    Index 0 Param 2=1
    Index 0 Param 3=1
    Module.Param0WriteEnabled=true
    Module.Param1WriteEnabled=true
    Module.Param2WriteEnabled=true
    Module.Param3WriteEnabled=true
   )
   &InheritSourceMovement(ApplyForce=false, ApplyPosition=true, ApplyPositionFalloff=false) {enabled=false}
   &Drag(Use Linear Drag=true, Use Rotational Drag=true)
   &CurlNoiseForce(Noise Frequency=0.25, Noise Strength=15, Pan Noise=false)
   &GravityForce(Gravity=&RandomRangeVector(Minimum=<0 0 -100>, Maximum=<0 0 -200>))
   &SolveForcesAndVelocity(Clamp Velocity=false, Limit Acceleration=false)
   /Niagara/VectorFields/ApplyVectorField(Vector Field Intensity as Force=0, Vector Field Intensity as Velocity=50, Vector Field Velocity Tightness=0.001) {enabled=false}
  ]
  rendererClasses [&NiagaraSpriteRendererProperties]
  renderers [{rendererClass=&NiagaraSpriteRendererProperties}]
 }
 Tinkerbell002 {
  simTarget=GPUComputeSim
  emitterUpdateScript [
   &EmitterState(Life Cycle Mode=::NewEnumerator1, Loop Behavior=::NewEnumerator1, Loop Delay=7, Loop Duration Mode=::NewEnumerator1, UseLoopDelay=false)
   /SetVariables(Emitter.Age:NiagaraFloat=&MakeCustomFloatFromBool(Bool=$System.PuffCanSpawnLeftFoot, True Float=$Emitter.Age, False Float=0)) {isSetParameters=true, module=SetVariables_#10}
   &SpawnRate(SpawnRate=20, Use Spawn Probability=false) {enabled=false}
   &SpawnBurst_Instantaneous(
    Age:NiagaraFloat=$Emitter.Age
    Spawn Count=5
    Spawn Probability=&ClampFloat(Float=&MakeFloatFromInt(Int=$System.LoopCount))
    Spawn Time=0.08
    Use Spawn Probability=true
    UseLoopCountLimit=false
   )
  ]
  particleSpawnScript [
   &PartitionParticles() {enabled=false}
   &InitializeParticle(
    Color=<rgba 0.5349 0.5662 1 0.49>
    Lifetime=0.75
    Position Offset=<0 0 100>
    Sprite Rotation Mode=::NewEnumerator2
    Sprite Size Mode=::NewEnumerator3
    Uniform Sprite Size=20
    UsePositionOffset=true
    Write Material Random=true
   )
   &SkeletalMeshLocation(Bone Sampling Mode=::NewEnumerator4, MaskSpawnGroup=false, Skeletal Mesh=$System.Skeletal Mesh)
   &ShapeLocation(
    Cylinder Height=4
    Cylinder Height Midpoint=0
    Cylinder Radius=15
    MaskSpawnGroup=false
    Offset=<0 0 -8>
    Offset Mode=::NewEnumerator0
    Override Seed=false
    Rotation Axis=<0 0 1>
    Rotation Mode=::NewEnumerator1
    Shape Origin=$Output.SkeletalMeshLocation.SampledPositionWithoutOffset
    Shape Primitive=::NewEnumerator1
    Surface Only=true
    Surface Only Band Thickness=0.5
    Use Endcaps In Surface Only Mode=false
    UseOwnerScale=true
   )
   &AddVelocity(
    Cone Angle=75
    Cone Axis=<0 0 1>
    Module.UseConeFalloff=true
    Module.UseVelDistribution=true
    Override Seed=false
    Velocity Mode=::NewEnumerator2
    Velocity Speed=&RandomRangeFloat(Minimum=75, Maximum=100)
   )
   /SetVariables(Particles.Velocity=&Multiply_VectorByFloat(
     Vector=&Multiply_Vector(A=$Particles.ShapeLocation.ShapeNormal, B=<1 1 500>)
     Float=&RandomRangeFloat(Minimum=10, Maximum=40)
    )) {enabled=false, isSetParameters=true, module=SetVariables_#11}
  ]
  particleUpdateScript [
   &SampleVectorField(
    FieldRotate:Quat4f=<0 0 1 1>
    Sampled Vector Scale=0.05
    VectorField=DataInterface() {propertyValues {Field=&Noisy_Turbulence_103, bTileX=true, bTileY=true, bTileZ=true}}
   ) {enabled=false}
   &ParticleState()
   &SpriteRotationRate(Rotation Rate=&MaskFloatBySpawnGroup(Masked Float=45, Passthrough Float=-45))
   &ScaleSpriteSize(Uniform Curve Sprite Scale=DataInterface() {propertyValues {Curve {keys [
        {arriveTangent=9.0725, interpMode=RCIM_Cubic, leaveTangent=9.0725, tangentMode=RCTM_User, time=0, value=0}
        {arriveTangent=-0.0039, interpMode=RCIM_Cubic, leaveTangent=-0.0039, tangentMode=RCTM_User, time=0.1408, value=0.9847}
        {arriveTangent=-4.4862, interpMode=RCIM_Cubic, leaveTangent=-4.4862, tangentMode=RCTM_User, time=1, value=0}
       ]}}})
   &ScaleColor(
    Scale Alpha=&FloatFromCurve(
     FloatCurve=DataInterface() {propertyValues {Curve {keys [{interpMode=RCIM_Cubic, tangentMode=RCTM_User, time=0, value=1}, {interpMode=RCIM_Cubic, tangentMode=RCTM_User, time=1, value=0}]}}}
     Scale Curve=&Waveform(
      Global Phase=&Add_Float(A=$Particles.Age, B=&RandomRangeFloat(Minimum=0, Maximum=1))
      "Waveform [1]"=::NewEnumerator0
      "[1] Frequency"=&RandomRangeFloat(Minimum=1.618, Maximum=3.618)
      "[1] Amplitude Min/Max"=<0 1>
     )
    )
    Scale RGB=&VectorFromFloat(Value=&Lerp_Float(
      A=1
      B=3
      Alpha=&FloatFromCurve(FloatCurve=DataInterface() {propertyValues {Curve {keys [<at 0 1>,<at 0.25 0>]}}})
     ))
    ScaleA=true
    ScaleRGB=true
   )
   &DynamicMaterialParameters(
    Index 0 Param 1=1
    Index 0 Param 2=1
    Index 0 Param 3=1
    Module.Param0WriteEnabled=true
    Module.Param1WriteEnabled=true
    Module.Param2WriteEnabled=true
    Module.Param3WriteEnabled=true
   )
   &InheritSourceMovement(ApplyForce=false, ApplyPosition=true, ApplyPositionFalloff=false) {enabled=false}
   &Drag(Use Linear Drag=true, Use Rotational Drag=true)
   &CurlNoiseForce(Noise Frequency=0.25, Noise Strength=15, Pan Noise=false)
   &GravityForce(Gravity=&RandomRangeVector(Minimum=<0 0 -100>, Maximum=<0 0 -200>))
   &SolveForcesAndVelocity(Clamp Velocity=false, Limit Acceleration=false)
   &ApplyVectorField(Vector Field Intensity as Force=0, Vector Field Intensity as Velocity=50, Vector Field Velocity Tightness=0.001) {enabled=false}
  ]
  rendererClasses [&NiagaraSpriteRendererProperties]
  renderers [{rendererClass=&NiagaraSpriteRendererProperties}]
 }
}
summary {
 systemName=NS_Emote_Marinara
 emitters [
  {
   emitterName=Empty001
   rendererClasses [
    &NiagaraSpriteRendererProperties
    &NiagaraLightRendererProperties
    &NiagaraComponentRendererProperties
   ]
  }
  {emitterName=Empty, rendererClasses [&NiagaraSpriteRendererProperties, &NiagaraLightRendererProperties]}
  {emitterName=SparkleRainTrail001, simTarget=GPUComputeSim, rendererClasses [&NiagaraSpriteRendererProperties]}
  {emitterName=Tinkerbell002, simTarget=GPUComputeSim, rendererClasses [&NiagaraSpriteRendererProperties]}
  {
   emitterName=Empty003
   rendererClasses [
    &NiagaraSpriteRendererProperties
    &NiagaraLightRendererProperties
    &NiagaraComponentRendererProperties
   ]
  }
  {
   emitterName=Empty002
   rendererClasses [
    &NiagaraSpriteRendererProperties
    &NiagaraLightRendererProperties
    &NiagaraComponentRendererProperties
   ]
  }
  {emitterName=SparkleRainTrail002, simTarget=GPUComputeSim, rendererClasses [&NiagaraSpriteRendererProperties]}
  {emitterName=Tinkerbell001, simTarget=GPUComputeSim, rendererClasses [&NiagaraSpriteRendererProperties]}
 ]
 userVariables [
  User.bIsFrontEnd {type=NiagaraBool}
  User.bIsFrontEndPreview {type=NiagaraBool}
  User.ConfettiColorArray {
   type=NiagaraDataInterfaceArrayColor
   description=""
   defaultValue {
    struct=NiagaraExt_VariableValue_DataInterface
    value {
     dataInterface=&self:NiagaraDataInterfaceArrayColor_1
     dataInterfaceClass=&NiagaraDataInterfaceArrayColor
    }
   }
  }
  User.FeetRestingPositions {
   type=NiagaraDataInterfaceArrayFloat3
   description=""
   defaultValue {
    struct=NiagaraExt_VariableValue_DataInterface
    value {
     dataInterface=&self:NiagaraDataInterfaceArrayFloat3_2
     dataInterfaceClass=&NiagaraDataInterfaceArrayFloat3
    }
   }
  }
  User.HandsSplash {type=NiagaraFloat, defaultValue=0, description=""}
  User.Index {type=NiagaraInt32}
  User.Local Alpha Preview {type=NiagaraFloat, defaultValue=0.75, description=""}
  User.LoopFoamSpawn {type=NiagaraFloat, defaultValue=2, description=""}
  User.Other Alpha Preview {type=NiagaraFloat, defaultValue=0.25, description=""}
  User.Ranked_Curve {type=NiagaraFloat, defaultValue=0, description=""}
  User.SongBPM {type=NiagaraFloat, defaultValue=97.5, description=""}
  User.SpotlightColorArray {
   type=NiagaraDataInterfaceArrayColor
   description=""
   defaultValue {
    struct=NiagaraExt_VariableValue_DataInterface
    value {
     dataInterface=&self:NiagaraDataInterfaceArrayColor_0
     dataInterfaceClass=&NiagaraDataInterfaceArrayColor
    }
   }
  }
  User.SpotlightsOrientationArray {
   type=NiagaraDataInterfaceArrayFloat3
   description=""
   defaultValue {
    struct=NiagaraExt_VariableValue_DataInterface
    value {
     dataInterface=&self:NiagaraDataInterfaceArrayFloat3_1
     dataInterfaceClass=&NiagaraDataInterfaceArrayFloat3
    }
   }
  }
  User.SpotlightsPositionArray {
   type=NiagaraDataInterfaceArrayFloat3
   description=""
   defaultValue {
    struct=NiagaraExt_VariableValue_DataInterface
    value {
     dataInterface=&self:NiagaraDataInterfaceArrayFloat3_0
     dataInterfaceClass=&NiagaraDataInterfaceArrayFloat3
    }
   }
  }
  User.Variant {type=NiagaraFloat, defaultValue=0, description=""}
  User.WaterPlaneLocation {type=Vector3f, description="", defaultValue {struct=Vector3f, value=<0 0 0>}}
  User.WaveAttenuationFactor {type=NiagaraFloat, defaultValue=0, description=""}
  User.WaveTime {type=NiagaraFloat, defaultValue=0, description=""}
 ]
}
