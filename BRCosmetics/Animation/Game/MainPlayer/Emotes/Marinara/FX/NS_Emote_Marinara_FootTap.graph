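# Niagara system for the foot-tap effect of the Marinara emote, producing a soft blue puff with a light flash.
# Its single emitter runs three short loops, burst-spawning at a sampled skeletal mesh bone or socket.
# Curves animate material parameters, size, alpha and light intensity, while sprites rotate and face world up.
# Sprite, light and component renderers are present; the right-foot gate on emitter age is disabled.
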NiagaraSystem /BRCosmetics/Animation/Game/MainPlayer/Emotes/Marinara/FX/NS_Emote_Marinara_FootTap
dependencies {
 usedDataInterfaces [&NiagaraDataInterfaceCurve, &NiagaraDataInterfaceSkeletalMesh]
 usedDynamicInputs [
  &Divide_Float
  &Add_Integer
  /Niagara/DynamicInputs/Divide/Divide_Int
  &ModuloInt
  &SetBoolByFloatComparison
  &MakeFloatFromVector
 ]
 usedModules [
  &self:SystemSpawnScript.SystemScriptSource.SystemScriptGraph.NiagaraNodeAssignment_3.SetVariables_#1
  &SystemState
  &self:SystemSpawnScript.SystemScriptSource.SystemScriptGraph.NiagaraNodeAssignment_0.SetVariables_#2
  &self:SystemSpawnScript.SystemScriptSource.SystemScriptGraph.NiagaraNodeAssignment_1.SetVariables_#3
  &self:SystemSpawnScript.SystemScriptSource.SystemScriptGraph.NiagaraNodeAssignment_2.SetVariables_#4
  &self:SystemSpawnScript.SystemScriptSource.SystemScriptGraph.NiagaraNodeAssignment_5.SetVariables_#5
 ]
}
emitters {Empty002 {
  emitterUpdateScript [
   &EmitterState(Life Cycle Mode=::NewEnumerator1, Loop Behavior=::NewEnumerator2, Loop Count=3, Loop Duration=0.95, UseLoopDelay=false)
   /SetVariables(Emitter.Age:NiagaraFloat=&MakeCustomFloatFromBool(Bool=$System.PuffCanSpawnRightFoot, True Float=$Emitter.Age, False Float=0)) {enabled=false, isSetParameters=true}
   &SpawnBurst_Instantaneous(
    Spawn Probability=&ClampFloat(Float=&MakeFloatFromInt(Int=$System.LoopCount))
    Use Spawn Probability=false
    UseLoopCountLimit=false
   )
  ]
  particleSpawnScript [
   &InitializeParticle(
    Color=<rgba 0.4914 0.5337 1 0.08>
    Lifetime=0.7
    Sprite Rotation Mode=::NewEnumerator2
    Sprite Size Mode=::NewEnumerator3
    Uniform Sprite Size=100
    UsePositionOffset=false
    Write Material Random=true
   )
   &SampleSkeletalMesh(
    AdditionalYawPitchRoll=false
    Bone / Socket Index=1
    Bone Sampling Mode:ENiagaraBoneSamplingMode=::NewEnumerator4
    Sampled Position Offset:Vector3f=<0 0 -2>
    Skeletal Mesh:NiagaraDataInterfaceSkeletalMesh=$System.Skeletal Mesh
   )
   &CameraOffset(Camera Offset Amount=5)
  ]
  particleUpdateScript [
   &ParticleState()
   &DynamicMaterialParameters(
    Index 0 Param 1=&FloatFromCurve(
     FloatCurve=DataInterface() {propertyValues {Curve {keys [{arriveTangent=1.3374, interpMode=RCIM_Cubic, leaveTangent=1.3374, tangentMode=RCTM_User, time=0, value=1}]}}}
     CurveIndex=$Particles.Age
    )
    Index 0 Param 2=0.15
    Index 0 Param 3=&FloatFromCurve(
     FloatCurve=DataInterface() {propertyValues {Curve {keys [
         {arriveTangent=4.9025, interpMode=RCIM_Cubic, leaveTangent=4.9025, tangentMode=RCTM_User, time=0, value=0}
         {arriveTangent=4.4266, interpMode=RCIM_Cubic, tangentMode=RCTM_Break, time=0.5796, value=2}
        ]}}}
     CurveIndex=$Particles.Age
     Scale Curve=0.05
    )
    Index 0 Param 4=&FloatFromCurve(
     FloatCurve=DataInterface() {propertyValues {Curve {keys [
         {arriveTangent=-3.3445, interpMode=RCIM_Cubic, leaveTangent=-3.3445, tangentMode=RCTM_User, time=0, value=0.462}
         {arriveTangent=-1.1268, interpMode=RCIM_Cubic, leaveTangent=-1.2066, tangentMode=RCTM_Break, time=0.3394, value=-0.3632}
         {arriveTangent=-1.4939, interpMode=RCIM_Cubic, tangentMode=RCTM_Break, time=1.2, value=-1}
        ]}}}
     CurveIndex=$Particles.Age
    )
    Index 1 Param 1=0.8
    Index 1 Param 2=0.5
    Index 1 Param 4=3
    Module.Param0WriteEnabled=true
    Module.Param0WriteEnabled1=true
    Module.Param1WriteEnabled=true
    Module.Param1WriteEnabled1=true
    Module.Param2WriteEnabled=true
    Module.Param2WriteEnabled1=true
    Module.Param3WriteEnabled=true
    Module.Param3WriteEnabled1=true
    Write Parameter Index 1=true
   )
   &SpriteRotationRate(Rotation Rate=20)
   /SetVariables(Particles.SpriteFacing=&TransformVector(Vector=<0 0 1>, Source Space=::World, Destination Space=::Local)) {isSetParameters=true}
   &ScaleSpriteSize(Uniform Curve Sprite Scale=DataInterface() {propertyValues {Curve {keys [
        {arriveTangent=4.1785, interpMode=RCIM_Cubic, leaveTangent=4.1785, tangentMode=RCTM_User, time=0, value=0}
        {arriveTangent=1.3778, interpMode=RCIM_Cubic, leaveTangent=1.3778, tangentMode=RCTM_User, time=0.1452, value=0.6}
        {arriveTangent=-0.0405, interpMode=RCIM_Cubic, leaveTangent=-0.0405, tangentMode=RCTM_User, time=1, value=1}
       ]}}})
   &ApplyOwnerScaleToAttributes(Scale Initial Sprite Size=true) {enabled=false}
   &SolveForcesAndVelocity(Clamp Velocity=false, Limit Acceleration=false)
   &ScaleColor(
    Scale Alpha=&FloatFromCurve(FloatCurve=DataInterface() {propertyValues {Curve {keys [
         {interpMode=RCIM_Cubic, tangentMode=RCTM_User, time=0, value=0}
         {interpMode=RCIM_Cubic, tangentMode=RCTM_User, time=0.5, value=1}
         {tangentMode=RCTM_User, time=1, value=0}
        ]}}})
    ScaleA=true
    ScaleRGB=true
   )
   /SetVariables(Particles.OffsetPosition:NiagaraPosition=&AddVectorToPosition(Position=$Particles.Position, Vector=$Particles.SampleSkeletalMesh.SampledPosition)) {isSetParameters=true}
   &FN_Light_Attributes(
    Color=<rgba 0.3437 0.4089 1 1>
    Create Color Attribute=true
    Create Position Attribute=true
    Create Radius Attribute=true
    Position=&AddVectorToPosition(Position=<0 0 15>, Vector=Linked() {linkedVariable {name=Particles.OffsetPosition, type=NiagaraPosition}})
    Radius=100
    Scale Color=25
   )
   /SetVariables(
    Particles.ColorLight:LinearColor=<rgba 0.3437 0.4089 1 1>
    Particles.LightIntensity:NiagaraFloat=&Lerp_Float(
     A=0
     B=35
     Alpha=&FloatFromCurve(FloatCurve=DataInterface() {propertyValues {Curve {keys [
          {arriveTangent=5.5909, interpMode=RCIM_Cubic, leaveTangent=5.5909, tangentMode=RCTM_User, time=0, value=0}
          {interpMode=RCIM_Cubic, tangentMode=RCTM_User, time=0.5, value=1}
          {arriveTangent=-5.2859, tangentMode=RCTM_User, time=1, value=0}
         ]}}})
    )
   ) {isSetParameters=true}
  ]
  rendererClasses [
   &NiagaraSpriteRendererProperties
   &NiagaraLightRendererProperties
   &NiagaraComponentRendererProperties
  ]
  renderers [
   {rendererClass=&NiagaraSpriteRendererProperties}
   {rendererClass=&NiagaraLightRendererProperties, rendererIndex=1}
   {rendererClass=&NiagaraComponentRendererProperties, rendererIndex=2}
   {rendererClass=&NiagaraSpriteRendererProperties, rendererIndex=3}
  ]
 }}
summary {
 systemName=NS_Emote_Marinara_FootTap
 emitters [{
   emitterName=Empty002
   rendererClasses [
    &NiagaraSpriteRendererProperties
    &NiagaraLightRendererProperties
    &NiagaraComponentRendererProperties
   ]
  }]
 userVariables [
  User.bIsFrontEnd {type=NiagaraBool}
  User.bIsFrontEndPreview {type=NiagaraBool}
  User.ConfettiColorArray {
   type=NiagaraDataInterfaceArrayColor
   description=""
   defaultValue {
    struct=NiagaraExt_VariableValue_DataInterface
    value {
     dataInterface=&self:NiagaraDataInterfaceArrayColor_1
     dataInterfaceClass=&NiagaraDataInterfaceArrayColor
    }
   }
  }
  User.FeetRestingPositions {
   type=NiagaraDataInterfaceArrayFloat3
   description=""
   defaultValue {
    struct=NiagaraExt_VariableValue_DataInterface
    value {
     dataInterface=&self:NiagaraDataInterfaceArrayFloat3_2
     dataInterfaceClass=&NiagaraDataInterfaceArrayFloat3
    }
   }
  }
  User.HandsSplash {type=NiagaraFloat, defaultValue=0, description=""}
  User.Index {type=NiagaraInt32}
  User.Local Alpha Preview {type=NiagaraFloat, defaultValue=0.75, description=""}
  User.LoopFoamSpawn {type=NiagaraFloat, defaultValue=2, description=""}
  User.Other Alpha Preview {type=NiagaraFloat, defaultValue=0.25, description=""}
  User.Ranked_Curve {type=NiagaraFloat, defaultValue=0, description=""}
  User.SongBPM {type=NiagaraFloat, defaultValue=97.5, description=""}
  User.SpotlightColorArray {
   type=NiagaraDataInterfaceArrayColor
   description=""
   defaultValue {
    struct=NiagaraExt_VariableValue_DataInterface
    value {
     dataInterface=&self:NiagaraDataInterfaceArrayColor_0
     dataInterfaceClass=&NiagaraDataInterfaceArrayColor
    }
   }
  }
  User.SpotlightsOrientationArray {
   type=NiagaraDataInterfaceArrayFloat3
   description=""
   defaultValue {
    struct=NiagaraExt_VariableValue_DataInterface
    value {
     dataInterface=&self:NiagaraDataInterfaceArrayFloat3_1
     dataInterfaceClass=&NiagaraDataInterfaceArrayFloat3
    }
   }
  }
  User.SpotlightsPositionArray {
   type=NiagaraDataInterfaceArrayFloat3
   description=""
   defaultValue {
    struct=NiagaraExt_VariableValue_DataInterface
    value {
     dataInterface=&self:NiagaraDataInterfaceArrayFloat3_0
     dataInterfaceClass=&NiagaraDataInterfaceArrayFloat3
    }
   }
  }
  User.Variant {type=NiagaraFloat, defaultValue=0, description=""}
  User.WaterPlaneLocation {type=Vector3f, description="", defaultValue {struct=Vector3f, value=<0 0 0>}}
  User.WaveAttenuationFactor {type=NiagaraFloat, defaultValue=0, description=""}
  User.WaveTime {type=NiagaraFloat, defaultValue=0, description=""}
 ]
}
